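# Sound cue for the looping first-person (1P) audio of the Interstellar emote.
# A looping wave player feeds an enveloper that quickly fades the volume in on a cubic curve, then a mixer whose second input is empty.
# It uses the 1P emote music sound class, emote attenuation and emote foley concurrency, and has a send to the music audio-analysis submix.
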SoundCue /BRCosmetics/Sounds/Emotes/Interstellar/Emote_Interstellar_Loop_1P_Cue
root=Mixer_0
nodes [
 {
  id=Enveloper_0
  children [WavePlayer_0]
  properties {
   pitchCurve {editorCurveData {keys [{time=0}]}}
   volumeCurve {editorCurveData {keys [{arriveTangent=10.8473, interpMode=RCIM_Cubic, leaveTangent=10.8473, tangentMode=RCTM_User, time=0, value=0}, <at 0.25 1>]}}
  }
 }
 {id=Mixer_0, children [Enveloper_0,null], properties {inputVolume [1,0.3]}}
 {id=WavePlayer_0, properties {bLooping=true, soundWaveAssetPtr=&/Emote_Interstellar_Loop}}
]
properties {
 attenuationSettings=&Fort_Emotes_Athena_Att
 maxDistance=1600
 soundClassObject=/Game/Sounds/FortSoundClassesAndModes/Fort_Music_Emote_1P_PSM
 volumeMultiplier=1
 concurrencySet [/Game/Sounds/Concurrency_Rules/Athena_EmoteFoley_Conc]
 soundSubmixSends [{minSendLevel=0, soundSubmix=&Fortnite_AudioAnalysis_Music}]
}
