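# Sound Cue for the looping third-person audio of the Interstellar emote, routed to the emote music sound class.
# A looping wave player feeds an Enveloper that quickly ramps volume up from silence, then a Mixer whose second input is empty.
# It uses emote attenuation and foley concurrency, keeps playing when silent, and sends to a music audio-analysis submix.
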SoundCue /BRCosmetics/Sounds/Emotes/Interstellar/SC_Emote_Interstellar_Loop_3P_Cue
root=Mixer_0
nodes [
 {
  id=Enveloper_1
  children [WavePlayer_1]
  properties {
   pitchCurve {editorCurveData {keys [{time=0}]}}
   volumeCurve {editorCurveData {keys [{arriveTangent=10.8473, interpMode=RCIM_Cubic, leaveTangent=10.8473, tangentMode=RCTM_User, time=0, value=0}, <at 0.25 1>]}}
  }
 }
 {id=Mixer_0, children [Enveloper_1,null], properties {inputVolume [1,0.3]}}
 {id=WavePlayer_1, properties {bLooping=true, soundWaveAssetPtr=&/Emote_Interstellar_Loop}}
]
properties {
 attenuationSettings=&Fort_Emotes_Athena_Att
 maxDistance=1600
 soundClassObject=/Game/Sounds/FortSoundClassesAndModes/Fort_Music_Emote_3P
 virtualizationMode=PlayWhenSilent
 volumeMultiplier=1
 concurrencySet [/Game/Sounds/Concurrency_Rules/Athena_EmoteFoley_Conc]
 soundSubmixSends [{minSendLevel=0, soundSubmix=&Fortnite_AudioAnalysis_Music}]
}
